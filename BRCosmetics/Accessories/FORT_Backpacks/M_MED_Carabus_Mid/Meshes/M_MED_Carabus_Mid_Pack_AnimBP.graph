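# Anim Blueprint for the Carabus Mid backpack cosmetic, adding physics motion to the pack.
# Each update copies parent-mesh root, pelvis and lower-spine world transforms onto a ref pose, then constrains pack rotation.
# RigidBody simulation in pack space varies gravity for skydiving and front end, is cached, and gets a second pass while sprinting.
# Weapon, emote and skydive events set material parameter "Hide Element 01", though every branch writes zero.
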AnimBlueprint /BRCosmetics/Accessories/FORT_Backpacks/M_MED_Carabus_Mid/Meshes/M_MED_Carabus_Mid_Pack_AnimBP
target_skeleton=&/M_MED_Carabus_Mid_Pack_Skeleton
graphs [
 AnimGraph {nodes [
   Copy Pose From Mesh() {bUseAttachedParent=true}
   n1 = RigidBody!(
    ComponentPose=Local To Component(LocalPose=Use cached pose 'initCache'())
    Alpha
    OverrideWorldGravity:Vector="0,0,-700"
    ExternalForce
    ComponentLinearAccScale:Vector="0.3,0.3,0.3"
    ComponentLinearVelScale:Vector="0.35,0.35,0.35"
   ) -> (Pose) {
    bClampLinearTranslationLimitToRefPose=true
    bOverrideWorldGravity=true
    overridePhysicsAsset=&/M_MED_Carabus_Mid_Pack_Physics
    simulationSpace=BaseBoneSpace
    alphaBoolBlend {blendOption=Linear}
    baseBoneRef {boneName=pack}
    componentAppliedLinearAccClamp=<2000 2000 2000>
    overrideWorldGravity {z=-700}
   }
   Output Pose(Result=Blend Poses by bool(
     BlendPose_0=Component To Local(ComponentPose=@n1)
     BlendPose_1=Use cached pose 'initCache'()
     BlendTime_0=0.5
     BlendTime_1=0.5
     bActiveValue=Get bIsSprinting()
    ) {blendType=Linear}) {layerGroup=None}
   n2 = "Transform (Modify) Bone - Bone: pack"(ComponentPose=Local To Component(LocalPose=Local Space Ref Pose()), Translation="0,0,3", Rotation="0,0,0", Scale="0.85,0.85,0.85") {
    scaleMode=BMM_Additive
    scaleSpace=BCS_BoneSpace
    translationMode=BMM_Additive
    translationSpace=BCS_BoneSpace
    alphaBoolBlend {blendOption=Linear}
    alphaScaleBiasClamp {bInterpResult=true, inRange {max=0, min=-40}, outRange {max=0, min=1}}
    boneToModify {boneName=pack}
    scale=<0.85 0.85 0.85>
    translation {z=3}
   }
   n3 = "Transform (Modify) Bone - Bone: body_root"(ComponentPose=@n2, Alpha=1, Translation=Get root_tran() -> (root_tran:Vector), Rotation=Get root_rot() -> (root_rot:Rotator), Scale="1,1,1") {
    rotationMode=BMM_Replace
    rotationSpace=BCS_WorldSpace
    scaleSpace=BCS_BoneSpace
    translationMode=BMM_Replace
    translationSpace=BCS_WorldSpace
    alphaBoolBlend {blendOption=Linear}
    alphaScaleBiasClamp {bInterpResult=true, inRange {max=0, min=-40}, outRange {max=0, min=1}}
    boneToModify {boneName=body_root}
   }
   n4 = "Transform (Modify) Bone - Bone: pelvis"(
    ComponentPose=@n3
    Alpha=1
    Translation=Get pelvis_tran() -> (pelvis_tran:Vector)
    Rotation=Get pelvis_rot() -> (pelvis_rot:Rotator)
    Scale="1,1,1"
   ) {
    rotationMode=BMM_Replace
    rotationSpace=BCS_WorldSpace
    scaleSpace=BCS_BoneSpace
    translationMode=BMM_Replace
    translationSpace=BCS_WorldSpace
    alphaBoolBlend {blendOption=Linear}
    alphaScaleBiasClamp {inRange {max=0, min=-40}, outRange {max=0, min=1}}
    boneToModify {boneName=pelvis}
   }
   n5 = "Transform (Modify) Bone - Bone: spine_01"(
    ComponentPose=@n4
    Alpha=1
    Translation=Get spine01_tran() -> (spine01_tran:Vector)
    Rotation=Get spine01_rot() -> (spine01_rot:Rotator)
    Scale="1,1,1"
   ) {
    rotationMode=BMM_Replace
    rotationSpace=BCS_WorldSpace
    scaleSpace=BCS_BoneSpace
    translationMode=BMM_Replace
    translationSpace=BCS_WorldSpace
    alphaBoolBlend {blendOption=Linear}
    alphaScaleBiasClamp {inRange {max=0, min=-40}, outRange {max=0, min=1}}
    boneToModify {boneName=spine_01}
   }
   n6 = "Transform (Modify) Bone - Bone: spine_02"(
    ComponentPose=@n5
    Alpha=1
    Translation=Get spine02_tran() -> (spine02_tran:Vector)
    Rotation=Get spine02_rot() -> (spine02_rot:Rotator)
    Scale="1,1,1"
   ) {
    rotationMode=BMM_Replace
    rotationSpace=BCS_WorldSpace
    scaleSpace=BCS_BoneSpace
    translationMode=BMM_Replace
    translationSpace=BCS_WorldSpace
    alphaBoolBlend {blendOption=Linear}
    alphaScaleBiasClamp {inRange {max=0, min=-40}, outRange {max=0, min=1}}
    boneToModify {boneName=spine_02}
   }
   n7 = "Constraint - pack (RR)"!(ComponentPose=@n6, Alpha=1, ConstraintWeights_0=1, ConstraintWeights_1:float=0) -> (Pose) {
    alphaBoolBlend {blendOption=Linear}
    boneToModify {boneName=pack}
    constraintSetup [{transformType=Rotation, targetBone {boneName=spine_05}}, {transformType=Rotation, targetBone {boneName=root}}]
    constraintWeights [1,1]
   }
   n8 = Get bIsInFrontEnd()
   n9 = OR Boolean!(A=Get bIsSkydiveFloating(), B=Get bIsSkyDiving(), C:bool) -> (ReturnValue)
   n10 = RigidBody(
    ComponentPose=@n7
    OverrideWorldGravity=SelectVector(A="10,0,200", B=SelectVector(A="0,0,-980", B="0,0,-700", bPickA=@n8), bPickA=@n9)
    ComponentLinearAccScale=SelectVector(A="0.4,0.4,0.4", B=SelectVector(A="0.05,0.05,0.05", B="0.3,0.35,0.35", bPickA=@n8), bPickA=@n9)
    ComponentLinearVelScale="0.35,0.35,0.35"
   ) {
    bClampLinearTranslationLimitToRefPose=true
    bOverrideWorldGravity=true
    overridePhysicsAsset=&M_MED_Carabus_Mid_Pack_Physics
    simulationSpace=BaseBoneSpace
    alphaBoolBlend {blendOption=Linear}
    baseBoneRef {boneName=pack}
    componentAppliedLinearAccClamp=<1200 1200 1200>
    overrideWorldGravity {z=-700}
   }
   Save cached pose 'initCache'(Pose=Component To Local(ComponentPose=@n10)) {cacheName=initCache}
  ]}
 EventGraph {nodes [
   Event OnWeaponChanged()
   > n1 = Get Dynamic Material Instance(SkeletalMeshComponent=GetOwningComponent(), ElementIndex=1)
   n2 = PickaxeItemDefinitionHasAnyMetaTag(Tags=Make Array("[0]"=(TagName="Cosmetics.MetaTags.CarabusFull")), bRequiresEquipped=true)
   > n3 = Branch(Condition=@n2)
   > n4 = Branch(Condition=Get bIsSkyDiving())
   n5 = Get Dynamic Material Instance(execute=@n16.then, SkeletalMeshComponent=GetOwningComponent(), ElementIndex=1)
   SetScalarParameterValue(execute=[@n6.then, @n7.then, @n4.then, @n3.else, @n9.else], self=[@n1, @n5], ParameterName=Hide Element 01, Value=0)
   SetScalarParameterValue(execute=[@n4.else, @n9.then], self=[@n1, @n5], ParameterName=Hide Element 01, Value=0)
   n6 = Event OnBeginPlayingEmote() {is=title K2Node_Event}
   n7 = Event OnBeginSkydiving()
   n8 = Event OnEndSkydiving()
   n9 = Branch(execute=[@n5.then, @n8.then], Condition=@n2)
   Event BlueprintUpdateAnimation()
   n10 = GetPartSkeletalMeshComponent()
   > Is Valid(InputObject=@n10)
   n11 = GetAttachParent!(self:Scene Component Object Reference=@n10) -> (ReturnValue:Scene Component Object Reference)
   > Is Valid(InputObject=@n11)
   n12 = BreakTransform(InTransform=GetSocketTransform(self=@n11, InSocketName=root, TransformSpace=RTS_World))
   > Set root_tran(root_tran:Vector=@n12) -> (Output_Get:Vector)
   > Set root_rot(root_rot:Rotator=@n12.Rotation) -> (Output_Get:Rotator)
   n13 = BreakTransform(InTransform=GetSocketTransform(self=@n11, InSocketName=pelvis, TransformSpace=RTS_World))
   > Set pelvis_tran(pelvis_tran:Vector=@n13) -> (Output_Get:Vector)
   > Set pelvis_rot(pelvis_rot:Rotator=@n13.Rotation) -> (Output_Get:Rotator)
   n14 = BreakTransform(InTransform=GetSocketTransform(self=@n11, InSocketName=spine_01, TransformSpace=RTS_World))
   > Set spine01_tran(spine01_tran:Vector=@n14) -> (Output_Get:Vector)
   > Set spine01_rot(spine01_rot:Rotator=@n14.Rotation) -> (Output_Get:Rotator)
   n15 = BreakTransform(InTransform=GetSocketTransform(self=@n11, InSocketName=spine_02, TransformSpace=RTS_World))
   > Set spine02_tran(spine02_tran:Vector=@n15) -> (Output_Get:Vector)
   > Set spine02_rot(spine02_rot:Rotator=@n15.Rotation) -> (Output_Get:Rotator)
   n16 = Event OnEndPlayingEmote() {is=title K2Node_Event}
   TryGetPawnOwner()
  ]}
]
variables [
 pelvis_rot {type=Rotator}
 pelvis_tran {type=Vector}
 root_rot {type=Rotator}
 root_tran {type=Vector}
 spine01_rot {type=Rotator}
 spine01_tran {type=Vector}
 spine02_rot {type=Rotator}
 spine02_tran {type=Vector}
]
